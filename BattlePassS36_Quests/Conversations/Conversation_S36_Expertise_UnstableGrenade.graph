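# Conversation database for the Season 36 Unstable Grenade weapon expertise quests, where an NPC pitches experimental Supernova Academy grenades for field testing and later offers a discount on them.
# Introduction and Finish entry tags each lead into a linear two-line speech chain.
# Each chain's first choice shows a main-quest icon and requires its matching intro or finished quest to be active.
# Every choice carries a RoseDepth conversation gameplay tag.
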ConversationDatabase /BattlePassS36_Quests/Conversations/Conversation_S36_Expertise_UnstableGrenade
entry_points [
 {tag=Conversation.Entry.Season36.Expertise.UnstableGrenade.Introduction, destinations [#1]}
 {tag=Conversation.Entry.Season36.Expertise.UnstableGrenade.Finish, destinations [#2]}
]
nodes [
 ConversationEntryPointNode_0 {
  id=#2
  class=ConversationEntryPointNode
  outputs [#3]
  properties {entryTag {tagName=Conversation.Entry.Season36.Expertise.UnstableGrenade.Finish}}
 }
 Speech_C_4 {
  id=#3
  class=Speech_C
  outputs [#4]
  properties {
   softTaskIcon=&T_UI_NPC_MainQuest
   generalConfig {
    defaultMessage="I am impressed. You have dexterity, grace, and somehow, all of your limbs."
    defaultMetadataParameters [IronLilac {value=""}]
   }
  }
  sub_nodes [
   ConversationChoiceNode_28 {
    class=ConversationChoiceNode
    properties {
     defaultChoiceDisplayText=Show off your Expertise
     choiceTags {gameplayTags [{tagName=Athena.Quests.S36.Conversations.RoseDepth.Finished.01}]}
    }
   }
   FortConversationRequirement_ControllerRequirement_28 {
    class=FortConversationRequirement_ControllerRequirement
    properties {
     nodeName=Require: Is Quest Active
     requirement {
      class=FortControllerRequirement_IsQuestActive
      properties {questDefinition=/BattlePassS36_Quests/Items/QuestItems/WeaponExpertise/UnstableGrenade/quest_S36_expertise_UnstableGrenade_finished}
     }
    }
   }
  ]
 }
 Speech_C_2 {
  id=#4
  class=Speech_C
  outputs []
  properties {generalConfig {defaultMessage=I can give you a 60% discount on these experimental grenades from our labs.}}
  sub_nodes [ConversationChoiceNode_16 {
    class=ConversationChoiceNode
    properties {defaultChoiceDisplayText=Continue, choiceTags {gameplayTags [{tagName=Athena.Quests.S36.Conversations.RoseDepth.Finished.02}]}}
   }]
 }
 ConversationEntryPointNode_3 {
  id=#1
  class=ConversationEntryPointNode
  outputs [#5]
  properties {entryTag {tagName=Conversation.Entry.Season36.Expertise.UnstableGrenade.Introduction}}
 }
 Speech_C_11 {
  id=#5
  class=Speech_C
  outputs [#6]
  properties {
   softTaskIcon=&T_UI_NPC_MainQuest
   generalConfig {
    defaultMessage=Supernova Academy is developing experimental grenades to tear apart the battlefield.
    defaultMetadataParameters [RoseDepth {value=""}]
   }
  }
  sub_nodes [
   ConversationChoiceNode_13 {
    class=ConversationChoiceNode
    properties {
     defaultChoiceDisplayText=Ask about Unstable Exotic Expertise
     choiceTags {gameplayTags [{tagName=Athena.Quests.S36.Conversations.RoseDepth.Intro.01}]}
    }
   }
   FortConversationRequirement_ControllerRequirement_10 {
    class=FortConversationRequirement_ControllerRequirement
    properties {
     nodeName=Require: Is Quest Active
     requirement {
      class=FortControllerRequirement_IsQuestActive
      properties {questDefinition=/BattlePassS36_Quests/Items/QuestItems/WeaponExpertise/UnstableGrenade/quest_S36_expertise_UnstableGrenade_intro}
     }
    }
   }
  ]
 }
 Speech_C_12 {
  id=#6
  class=Speech_C
  outputs []
  properties {generalConfig {defaultMessage=We can share them with you to field test... for a small price.}}
  sub_nodes [ConversationChoiceNode_14 {
    class=ConversationChoiceNode
    properties {defaultChoiceDisplayText=Continue, choiceTags {gameplayTags [{tagName=Athena.Quests.S36.Conversations.RoseDepth.Intro.02}]}}
   }]
 }
]
