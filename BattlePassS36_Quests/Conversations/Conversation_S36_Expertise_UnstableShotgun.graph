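# Conversation database for the Season 36 Unstable Shotgun weapon expertise quest, providing NPC dialogue for its introduction and finish.
# Two entry tags each lead to a two-step speech chain, with the first speech gated on the matching intro or finished quest being active.
# Choices carry IronLilac conversation gameplay tags for each step.
# The finish grants a spoken certificate and discount, while the second intro line is still placeholder text.
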ConversationDatabase /BattlePassS36_Quests/Conversations/Conversation_S36_Expertise_UnstableShotgun
entry_points [
 {tag=Conversation.Entry.Season36.Expertise.UnstableShotgun.Introduction, destinations [#1]}
 {tag=Conversation.Entry.Season36.Expertise.UnstableShotgun.Finish, destinations [#2]}
]
nodes [
 ConversationEntryPointNode_0 {
  id=#2
  class=ConversationEntryPointNode
  outputs [#3]
  properties {entryTag {tagName=Conversation.Entry.Season36.Expertise.UnstableShotgun.Finish}}
 }
 Speech_C_4 {
  id=#3
  class=Speech_C
  outputs [#4]
  properties {
   softTaskIcon=&T_UI_NPC_MainQuest
   generalConfig {defaultMessage="Looks like you're top of the class! Well done, expert.", defaultMetadataParameters [IronLilac {value=""}]}
  }
  sub_nodes [
   ConversationChoiceNode_28 {
    class=ConversationChoiceNode
    properties {
     defaultChoiceDisplayText=Show off your Expertise
     choiceTags {gameplayTags [{tagName=Athena.Quests.S36.Conversations.IronLilac.Finished.01}]}
    }
   }
   FortConversationRequirement_ControllerRequirement_28 {
    class=FortConversationRequirement_ControllerRequirement
    properties {
     nodeName=Require: Is Quest Active
     requirement {
      class=FortControllerRequirement_IsQuestActive
      properties {questDefinition=/BattlePassS36_Quests/Items/QuestItems/WeaponExpertise/UnstableShotgun/quest_S36_expertise_UnstableShotgun_finished}
     }
    }
   }
  ]
 }
 Speech_C_2 {
  id=#4
  class=Speech_C
  outputs []
  properties {generalConfig {defaultMessage=Please enjoy this honorary certificate from Supernova Academy... and a 60% discount.}}
  sub_nodes [ConversationChoiceNode_16 {
    class=ConversationChoiceNode
    properties {defaultChoiceDisplayText=Continue, choiceTags {gameplayTags [{tagName=Athena.Quests.S36.Conversations.IronLilac.Finished.02}]}}
   }]
 }
 ConversationEntryPointNode_3 {
  id=#1
  class=ConversationEntryPointNode
  outputs [#5]
  properties {entryTag {tagName=Conversation.Entry.Season36.Expertise.UnstableShotgun.Introduction}}
 }
 Speech_C_11 {
  id=#5
  class=Speech_C
  outputs [#6]
  properties {
   softTaskIcon=&T_UI_NPC_MainQuest
   generalConfig {
    defaultMessage=You know your way around a shotgun. We're looking for experts to test some prototypes.
    defaultMetadataParameters [IronLilac {value=""}]
   }
  }
  sub_nodes [
   ConversationChoiceNode_13 {
    class=ConversationChoiceNode
    properties {
     defaultChoiceDisplayText=Ask about Unstable Exotic Expertise
     choiceTags {gameplayTags [{tagName=Athena.Quests.S36.Conversations.IronLilac.Intro.01}]}
    }
   }
   FortConversationRequirement_ControllerRequirement_10 {
    class=FortConversationRequirement_ControllerRequirement
    properties {
     nodeName=Require: Is Quest Active
     requirement {
      class=FortControllerRequirement_IsQuestActive
      properties {questDefinition=/BattlePassS36_Quests/Items/QuestItems/WeaponExpertise/UnstableShotgun/quest_S36_expertise_UnstableShotgun_intro}
     }
    }
   }
  ]
 }
 Speech_C_12 {
  id=#6
  class=Speech_C
  outputs []
  properties {generalConfig {defaultMessage="[Default Message]"}}
  sub_nodes [ConversationChoiceNode_14 {
    class=ConversationChoiceNode
    properties {defaultChoiceDisplayText=Continue, choiceTags {gameplayTags [{tagName=Athena.Quests.S36.Conversations.IronLilac.Intro.02}]}}
   }]
 }
]
