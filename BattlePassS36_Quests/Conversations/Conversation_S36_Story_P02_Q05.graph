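# Season 36 story conversation (P02 Q05) in which an NPC catches a spy, reveals that a vault is meant for the Zero Point Shard and advises secrecy.
# A single entry tag leads into three linear speech nodes.
# The first shows a main-quest icon, offers a tagged "?!?" choice and requires quest Q05b to be active.
# The other two use Continue choices, and the last ends the conversation.
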ConversationDatabase /BattlePassS36_Quests/Conversations/Conversation_S36_Story_P02_Q05
entry_points [{tag=Conversation.Entry.Season36.Story.P02.Q05, destinations [#1]}]
nodes [
 ConversationEntryPointNode_1 {
  id=#1
  class=ConversationEntryPointNode
  outputs [#2]
  properties {entryTag {tagName=Conversation.Entry.Season36.Story.P02.Q05}}
 }
 Speech_C_10 {
  id=#2
  class=Speech_C
  outputs [#3]
  properties {
   softTaskIcon=&T_UI_NPC_MainQuest
   generalConfig {defaultMessage="It seems I have a shadow. You're lucky I've always had a fondness for shadows, Myst."}
  }
  sub_nodes [
   ConversationChoiceNode_2 {
    class=ConversationChoiceNode
    properties {defaultChoiceDisplayText="?!?", choiceTags {gameplayTags [{tagName=Athena.Quests.S36.Conversations.Story.P02.Selection}]}}
   }
   FortConversationRequirement_ControllerRequirement_1 {
    class=FortConversationRequirement_ControllerRequirement
    properties {
     nodeName=Require: Is Quest Active
     requirement {
      class=FortControllerRequirement_IsQuestActive
      properties {questDefinition=/BattlePassS36_Quests/Items/QuestItems/Story/P02/Quest_S36_Story_P02_Q05b}
     }
    }
   }
  ]
 }
 Speech_C_0 {
  id=#3
  class=Speech_C
  outputs [#4]
  properties {
   generalConfig {defaultMessage="Yes, this vault is intended to hold the Zero Point Shard, once we have collected it."}
   voiceOverTag {tagName=None}
  }
  sub_nodes [ConversationChoiceNode_3 {class=ConversationChoiceNode, properties {defaultChoiceDisplayText=Continue}}]
 }
 Speech_C_2 {
  id=#4
  class=Speech_C
  outputs []
  properties {
   generalConfig {defaultMessage="You're not a bad spy, if inexperienced. Your first lesson in espionage: know when to keep a secret."}
   voiceOverTag {tagName=None}
  }
  sub_nodes [ConversationChoiceNode_4 {class=ConversationChoiceNode, properties {defaultChoiceDisplayText=Continue}}]
 }
]
