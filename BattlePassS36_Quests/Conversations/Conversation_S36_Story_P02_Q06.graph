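# Conversation database for Season 36 story quest P02 Q06, a linear NPC dialogue about Midas funding Supernova Academy and the Vault to secure the Shard.
# A tagged entry point leads to a first speech offered as a "Confront" choice with a main quest icon, gated on that quest being active.
# Three further speech nodes follow in sequence, each advanced by a Continue choice, and the last ends the conversation.
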ConversationDatabase /BattlePassS36_Quests/Conversations/Conversation_S36_Story_P02_Q06
entry_points [{tag=Conversation.Entry.Season36.Story.P02.Q06, destinations [#1]}]
nodes [
 ConversationEntryPointNode_1 {
  id=#1
  class=ConversationEntryPointNode
  outputs [#2]
  properties {entryTag {tagName=Conversation.Entry.Season36.Story.P02.Q06}}
 }
 Speech_C_10 {
  id=#2
  class=Speech_C
  outputs [#3]
  properties {
   softTaskIcon=&T_UI_NPC_MainQuest
   generalConfig {defaultMessage="I'm impressed. Yeah, Midas is bankrolling Supernova Academy... and our other special project, the Vault."}
  }
  sub_nodes [
   ConversationChoiceNode_2 {
    class=ConversationChoiceNode
    properties {defaultChoiceDisplayText=Confront, choiceTags {gameplayTags [{tagName=Athena.Quests.S36.Conversations.Story.P02.Q6Hope}]}}
   }
   FortConversationRequirement_ControllerRequirement_1 {
    class=FortConversationRequirement_ControllerRequirement
    properties {
     nodeName=Require: Is Quest Active
     requirement {
      class=FortControllerRequirement_IsQuestActive
      properties {questDefinition=/BattlePassS36_Quests/Items/QuestItems/Story/P02/Quest_S36_Story_P02_Q06}
     }
    }
   }
  ]
 }
 Speech_C_0 {
  id=#3
  class=Speech_C
  outputs [#4]
  properties {generalConfig {defaultMessage="I know what you're thinking, and I agree with you. Trusting Midas is a risk."}, voiceOverTag {tagName=None}}
  sub_nodes [ConversationChoiceNode_3 {class=ConversationChoiceNode, properties {defaultChoiceDisplayText=Continue}}]
 }
 Speech_C_2 {
  id=#4
  class=Speech_C
  outputs [#5]
  properties {
   generalConfig {defaultMessage="But NOT securing the Shard is an even bigger risk, and I don't mean just to this Island."}
   voiceOverTag {tagName=None}
  }
  sub_nodes [ConversationChoiceNode_4 {class=ConversationChoiceNode, properties {defaultChoiceDisplayText=Continue}}]
 }
 Speech_C_3 {
  id=#5
  class=Speech_C
  outputs []
  properties {
   generalConfig {defaultMessage="Plus, if there are more Shards out there, causing problems... We'll REALLY need the Vault."}
   voiceOverTag {tagName=None}
  }
  sub_nodes [ConversationChoiceNode_1 {class=ConversationChoiceNode, properties {defaultChoiceDisplayText=Continue}}]
 }
]
