ConversationDatabase /BattlePassS36_Quests/Conversations/Conversation_S36_Story_P03_Q06
entry_points [{tag=Conversation.Entry.Season36.IronLilac.Story.P3Q6, destinations [#1]}]
nodes [
 ConversationEntryPointNode_1 {
  id=#1
  class=ConversationEntryPointNode
  outputs [#2]
  properties {entryTag {tagName=Conversation.Entry.Season36.IronLilac.Story.P3Q6}}
 }
 Speech_C_10 {
  id=#2
  class=Speech_C
  outputs [#3]
  properties {
   softTaskIcon=&T_UI_NPC_MainQuest
   generalConfig {defaultMessage="After the Academy fell, I was worried that this initiative, The Supernovas, it would all fall apart."}
  }
  sub_nodes [
   ConversationChoiceNode_2 {
    class=ConversationChoiceNode
    properties {defaultChoiceDisplayText=Check In, choiceTags {gameplayTags [{tagName=Athena.Quests.S36.Conversations.Story.P03.HopeQ6}]}}
   }
   FortConversationRequirement_ControllerRequirement_1 {
    class=FortConversationRequirement_ControllerRequirement
    properties {
     nodeName=Require: Is Quest Active
     requirement {
      class=FortControllerRequirement_IsQuestActive
      properties {questDefinition=/BattlePassS36_Quests/Items/QuestItems/Story/P03/Quest_S36_Story_P03_Q06}
     }
    }
   }
  ]
 }
 Speech_C_0 {
  id=#3
  class=Speech_C
  outputs [#4]
  properties {generalConfig {defaultMessage="But you and Morgan have made sure we stand tall, as a team."}, voiceOverTag {tagName=None}}
  sub_nodes [ConversationChoiceNode_3 {class=ConversationChoiceNode, properties {defaultChoiceDisplayText=Continue}}]
 }
 Speech_C_2 {
  id=#4
  class=Speech_C
  outputs []
  properties {generalConfig {defaultMessage=Thank you.}, voiceOverTag {tagName=None}}
  sub_nodes [ConversationChoiceNode_4 {class=ConversationChoiceNode, properties {defaultChoiceDisplayText=Continue}}]
 }
]
