ConversationDatabase /BattlePassS36_Quests/Conversations/NPCLibrary_Conversation_Superman_Quests
entry_points [{tag=Conversation.Entry.IcicleLarva.Quests, destinations [#1]}]
nodes [
 ConversationEntryPointNode_3 {
  id=#1
  class=ConversationEntryPointNode
  outputs [#2]
  properties {entryTag {tagName=Conversation.Entry.IcicleLarva.Quests}}
 }
 Speech_C_12 {
  id=#2
  class=Speech_C
  outputs [#3]
  properties {
   softTaskIcon=&T_UI_NPC_MainQuest
   generalConfig {defaultMessage="Thank you! We can examine these to learn more about The Mask Maker's plans."}
  }
  sub_nodes [
   ConversationChoiceNode_4 {
    class=ConversationChoiceNode
    properties {defaultChoiceDisplayText=Report, choiceTags {gameplayTags [{tagName=Athena.Quests.S36.Conversations.Story.P01.Q6Supes}]}}
   }
   FortConversationRequirement_ControllerRequirement_1 {
    class=FortConversationRequirement_ControllerRequirement
    properties {
     nodeName=Require: Is Quest Active
     requirement {
      class=FortControllerRequirement_IsQuestActive
      properties {questDefinition=/BattlePassS36_Quests/Items/QuestItems/Story/P01/Quest_S36_Story_P01_Q06}
     }
    }
   }
  ]
 }
 Speech_C_11 {
  id=#3
  class=Speech_C
  outputs []
  properties {generalConfig {defaultMessage="If we can understand what's coming, we can protect people from these Spires."}, voiceOverTag {tagName=None}}
  sub_nodes [ConversationChoiceNode_3 {class=ConversationChoiceNode, properties {defaultChoiceDisplayText=Continue}}]
 }
]
